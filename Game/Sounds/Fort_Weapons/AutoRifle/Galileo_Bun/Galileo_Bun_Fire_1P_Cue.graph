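# Sound cue for the first-person firing sound of the Galileo_Bun auto rifle.
# A random node picks one of four equally weighted close-range fire recordings and feeds it through a modulator with pitch held at unity.
# It uses first-person projectile attenuation, a ranged-fire 1P sound class and a first-person gunfire concurrency set.
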SoundCue /Game/Sounds/Fort_Weapons/AutoRifle/Galileo_Bun/Galileo_Bun_Fire_1P_Cue
root=Modulator_1
nodes [
 {id=Modulator_1, children [Random_0], properties {pitchMax=1, pitchMin=1}}
 {id=Random_0, children [WavePlayer_8,WavePlayer_0,WavePlayer_6,WavePlayer_7], properties {weights [1,1,1,1]}}
 {id=WavePlayer_0, properties {soundWaveAssetPtr=&/Galileo_Bun_Fire_Close_03}}
 {id=WavePlayer_6, properties {soundWaveAssetPtr=&/Galileo_Bun_Fire_Close_04}}
 {id=WavePlayer_7, properties {soundWaveAssetPtr=&/Galileo_Bun_Fire_Close_05}}
 {id=WavePlayer_8, properties {soundWaveAssetPtr=&/Galileo_Bun_Fire_Close_01}}
]
properties {
 attenuationSettings=&Fort_Weapon_Projectile_FirstPerson_Athena_Att
 maxDistance=2097152
 soundClassObject=&Fort_WeaponRanged_Fire_1P_PSM
 volumeMultiplier=1
 attenuationOverrides {falloffDistance=5000, attenuationShapeExtents {x=1000}}
 concurrencyOverrides {bLimitToOwner=true, maxCount=4, resolutionRule=StopOldest}
 concurrencySet [&Weapon_Gunfire_FirstPerson_Conc]
}
