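# Sound cue for Galileo_Bun auto rifle gunfire, with concurrency limited by local and global gunfire settings.
# Each quality branch mixes close, medium and distant layers, each with its own attenuation and a sound class (third-person for close, Distant for others).
# A quality switch steps down from larger random sample pools to smaller ones.
# The lowest two levels share single-sample layers with random pitch variation to offset repetition.
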SoundCue /Game/Sounds/Fort_Weapons/AutoRifle/Galileo_Bun/Galileo_Bun_Fire_Cue
root=QualityLevel_0
nodes [
 {id=Attenuation_0, children [Modulator_2], properties {attenuationSettings=&Fort_Weapon_Projectile_Close_Athena_Att}}
 {
  id=Attenuation_1
  children [Modulator_4]
  properties {attenuationSettings=&Fort_Weapon_Projectile_Medium_Mono_Athena_Att}
 }
 {
  id=Attenuation_118
  children [Modulator_0]
  properties {attenuationSettings=&Fort_Weapon_Projectile_Close_Athena_Att}
 }
 {
  id=Attenuation_119
  children [Modulator_1]
  properties {attenuationSettings=&Fort_Weapon_Projectile_Medium_Mono_Athena_Att}
 }
 {
  id=Attenuation_120
  children [Modulator_12]
  properties {attenuationSettings=&Fort_Weapon_Projectile_Distant_Mono_Athena_Att}
 }
 {
  id=Attenuation_2
  children [Modulator_3]
  properties {attenuationSettings=&Fort_Weapon_Projectile_Distant_Mono_Athena_Att}
 }
 {id=Attenuation_3, children [Modulator_5], properties {attenuationSettings=&Fort_Weapon_Projectile_Close_Athena_Att}}
 {
  id=Attenuation_4
  children [Modulator_7]
  properties {attenuationSettings=&Fort_Weapon_Projectile_Medium_Mono_Athena_Att}
 }
 {
  id=Attenuation_5
  children [Modulator_6]
  properties {attenuationSettings=&Fort_Weapon_Projectile_Distant_Mono_Athena_Att}
 }
 {id=Mixer_0, children [SoundClass_2,SoundClass_1,SoundClass_3], properties {inputVolume [1,1,0.8]}}
 {id=Mixer_1, children [SoundClass_5,SoundClass_4,SoundClass_6], properties {inputVolume [1,1,0.8]}}
 {id=Mixer_38, children [SoundClass_77, SoundClass_76, SoundClass_0], properties {inputVolume [1,1,0.8]}}
 {id=Modulator_0, children [Random_0], properties {pitchMax=1, pitchMin=1}}
 {id=Modulator_1, children [Random_1], properties {pitchMax=1, pitchMin=1, volumeMax=1, volumeMin=1}}
 {id=Modulator_12, children [Random_2], properties {pitchMax=1, pitchMin=1, volumeMax=0.9, volumeMin=0.9}}
 {id=Modulator_2, children [Random_3], properties {pitchMax=1, pitchMin=1}}
 {id=Modulator_3, children [Random_5], properties {pitchMax=1, pitchMin=1, volumeMax=0.9, volumeMin=0.9}}
 {id=Modulator_4, children [Random_4], properties {pitchMax=1, pitchMin=1, volumeMax=1, volumeMin=1}}
 {id=Modulator_5, children [WavePlayer_33], properties {pitchMax=1, pitchMin=1}}
 {id=Modulator_6, children [WavePlayer_39], properties {pitchMax=1, pitchMin=1, volumeMax=0.9, volumeMin=0.9}}
 {id=Modulator_7, children [WavePlayer_36], properties {pitchMax=1, pitchMin=1, volumeMax=1, volumeMin=1}}
 {id=Modulator_8, children [Mixer_1], properties {pitchMax=1.1, pitchMin=0.9}}
 {id=QualityLevel_0, children [Mixer_38,Mixer_0,Modulator_8,Modulator_8]}
 {id=Random_0, children [WavePlayer_0,WavePlayer_2,WavePlayer_5,WavePlayer_7], properties {weights [1,1,1,1]}}
 {id=Random_1, children [WavePlayer_8, WavePlayer_9, WavePlayer_10, WavePlayer_11, WavePlayer_12], properties {weights [1,1,1,1,1]}}
 {id=Random_2, children [WavePlayer_13, WavePlayer_14, WavePlayer_15, WavePlayer_16, WavePlayer_17], properties {weights [1,1,1,1,1]}}
 {id=Random_3, children [WavePlayer_18, WavePlayer_19, WavePlayer_22], properties {weights [1,1,1]}}
 {id=Random_4, children [WavePlayer_23, WavePlayer_24, WavePlayer_25], properties {weights [1,1,1]}}
 {id=Random_5, children [WavePlayer_28, WavePlayer_29, WavePlayer_30], properties {weights [1,1,1]}}
 {
  id=SoundClass_0
  children [Attenuation_120]
  properties {soundClassOverride=&Fort_WeaponRanged_Fire_Distant}
 }
 {
  id=SoundClass_1
  children [Attenuation_1]
  properties {soundClassOverride=&Fort_WeaponRanged_Fire_Distant}
 }
 {
  id=SoundClass_2
  children [Attenuation_0]
  properties {soundClassOverride=&Fort_WeaponRanged_Fire_3P_PSM}
 }
 {
  id=SoundClass_3
  children [Attenuation_2]
  properties {soundClassOverride=&Fort_WeaponRanged_Fire_Distant}
 }
 {
  id=SoundClass_4
  children [Attenuation_4]
  properties {soundClassOverride=&Fort_WeaponRanged_Fire_Distant}
 }
 {
  id=SoundClass_5
  children [Attenuation_3]
  properties {soundClassOverride=&Fort_WeaponRanged_Fire_3P_PSM}
 }
 {
  id=SoundClass_6
  children [Attenuation_5]
  properties {soundClassOverride=&Fort_WeaponRanged_Fire_Distant}
 }
 {
  id=SoundClass_76
  children [Attenuation_119]
  properties {soundClassOverride=&Fort_WeaponRanged_Fire_Distant}
 }
 {
  id=SoundClass_77
  children [Attenuation_118]
  properties {soundClassOverride=&Fort_WeaponRanged_Fire_3P_PSM}
 }
 {id=WavePlayer_0, properties {soundWaveAssetPtr=&/Galileo_Bun_Fire_Close_01}}
 {id=WavePlayer_10, properties {soundWaveAssetPtr=&/Galileo_Bun_Fire_Medium_03}}
 {id=WavePlayer_11, properties {soundWaveAssetPtr=&/Galileo_Bun_Fire_Medium_04}}
 {id=WavePlayer_12, properties {soundWaveAssetPtr=&/Galileo_Bun_Fire_Medium_05}}
 {id=WavePlayer_13, properties {soundWaveAssetPtr=&/Galileo_Bun_Fire_Distant_01}}
 {id=WavePlayer_14, properties {soundWaveAssetPtr=&/Galileo_Bun_Fire_Distant_02}}
 {id=WavePlayer_15, properties {soundWaveAssetPtr=&/Galileo_Bun_Fire_Distant_03}}
 {id=WavePlayer_16, properties {soundWaveAssetPtr=&/Galileo_Bun_Fire_Distant_04}}
 {id=WavePlayer_17, properties {soundWaveAssetPtr=&/Galileo_Bun_Fire_Distant_05}}
 {id=WavePlayer_18, properties {soundWaveAssetPtr=&Galileo_Bun_Fire_Close_01}}
 {id=WavePlayer_19, properties {soundWaveAssetPtr=&/Galileo_Bun_Fire_Close_02}}
 {id=WavePlayer_2, properties {soundWaveAssetPtr=&/Galileo_Bun_Fire_Close_03}}
 {id=WavePlayer_22, properties {soundWaveAssetPtr=&/Galileo_Bun_Fire_Close_05}}
 {id=WavePlayer_23, properties {soundWaveAssetPtr=&/Galileo_Bun_Fire_Medium_01}}
 {id=WavePlayer_24, properties {soundWaveAssetPtr=&/Galileo_Bun_Fire_Medium_02}}
 {id=WavePlayer_25, properties {soundWaveAssetPtr=&Galileo_Bun_Fire_Medium_03}}
 {id=WavePlayer_28, properties {soundWaveAssetPtr=&Galileo_Bun_Fire_Distant_01}}
 {id=WavePlayer_29, properties {soundWaveAssetPtr=&Galileo_Bun_Fire_Distant_02}}
 {id=WavePlayer_30, properties {soundWaveAssetPtr=&Galileo_Bun_Fire_Distant_03}}
 {id=WavePlayer_33, properties {soundWaveAssetPtr=&Galileo_Bun_Fire_Close_01}}
 {id=WavePlayer_36, properties {soundWaveAssetPtr=&Galileo_Bun_Fire_Medium_01}}
 {id=WavePlayer_39, properties {soundWaveAssetPtr=&Galileo_Bun_Fire_Distant_01}}
 {id=WavePlayer_5, properties {soundWaveAssetPtr=&/Galileo_Bun_Fire_Close_04}}
 {id=WavePlayer_7, properties {soundWaveAssetPtr=&Galileo_Bun_Fire_Close_05}}
 {id=WavePlayer_8, properties {soundWaveAssetPtr=&Galileo_Bun_Fire_Medium_01}}
 {id=WavePlayer_9, properties {soundWaveAssetPtr=&Galileo_Bun_Fire_Medium_02}}
]
properties {
 bDebug=true
 maxDistance=27500
 volumeMultiplier=0.9
 concurrencyOverrides {resolutionRule=PreventNew}
 concurrencySet [&Weapon_Gunfire_Local_Conc, &Weapon_Gunfire_Global_Conc]
}
