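# Sound cue for body impacts of the Galileo_Bun auto rifle weapon, classed under Fort_Impacts.
# A Random node picks one of six equally weighted impact waves and feeds it through a Modulator at default settings for variation.
# The cue uses the shared bullet impact attenuation and overrides concurrency to cap simultaneous instances at a low count.
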SoundCue /Game/Sounds/Fort_Weapons/AutoRifle/Galileo_Bun/Galileo_Bun_Impact_Body_Cue
root=Modulator_1
nodes [
 {id=Modulator_1, children [Random_1]}
 {
  id=Random_1
  children [WavePlayer_0,WavePlayer_1,WavePlayer_2,WavePlayer_3,WavePlayer_4,WavePlayer_5]
  properties {bShouldExcludeFromBranchCulling=true, weights [1,1,1,1,1,1]}
 }
 {id=WavePlayer_0, properties {soundWaveAssetPtr=&/Galileo_Bun_Impact_Body_01}}
 {id=WavePlayer_1, properties {soundWaveAssetPtr=&/Galileo_Bun_Impact_Body_02}}
 {id=WavePlayer_2, properties {soundWaveAssetPtr=&/Galileo_Bun_Impact_Body_03}}
 {id=WavePlayer_3, properties {soundWaveAssetPtr=&/Galileo_Bun_Impact_Body_04}}
 {id=WavePlayer_4, properties {soundWaveAssetPtr=&/Galileo_Bun_Impact_Body_05}}
 {id=WavePlayer_5, properties {soundWaveAssetPtr=&/Galileo_Bun_Impact_Body_06}}
]
properties {
 attenuationSettings=/Game/Sounds/Attenuation/Fort_Bullet_Impact_Att
 bOverrideConcurrency=true
 maxDistance=5300
 soundClassObject=&Fort_Impacts
 volumeMultiplier=1
 concurrencyOverrides {maxCount=3}
}
