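# Sound cue for Galileo_Bun auto rifle bullet impacts on surfaces.
# A random node picks one of six equally weighted impact waves, with a modulator varying pitch and a very short random delay offsetting playback.
# It plays in the impacts sound class with shared bullet-impact attenuation, and its own concurrency override, which takes precedence over the assigned rifle impact concurrency set, caps simultaneous voices.
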SoundCue /Game/Sounds/Fort_Weapons/AutoRifle/Galileo_Bun/Galileo_Bun_Impact_Surface_Cue
root=Delay_0
nodes [
 {id=Delay_0, children [Modulator_0], properties {delayMax=0.02}}
 {id=Modulator_0, children [Random_0], properties {pitchMax=1.1}}
 {
  id=Random_0
  children [WavePlayer_0,WavePlayer_1,WavePlayer_2,WavePlayer_3,WavePlayer_4,WavePlayer_5]
  properties {bShouldExcludeFromBranchCulling=true, weights [1,1,1,1,1,1]}
 }
 {id=WavePlayer_0, properties {soundWaveAssetPtr=&/Galileo_Bun_Impact_Surface_01}}
 {id=WavePlayer_1, properties {soundWaveAssetPtr=&/Galileo_Bun_Impact_Surface_02}}
 {id=WavePlayer_2, properties {soundWaveAssetPtr=&/Galileo_Bun_Impact_Surface_03}}
 {id=WavePlayer_3, properties {soundWaveAssetPtr=&/Galileo_Bun_Impact_Surface_04}}
 {id=WavePlayer_4, properties {soundWaveAssetPtr=&/Galileo_Bun_Impact_Surface_05}}
 {id=WavePlayer_5, properties {soundWaveAssetPtr=&/Galileo_Bun_Impact_Surface_06}}
]
properties {
 attenuationSettings=/Game/Sounds/Attenuation/Fort_Bullet_Impact_Att
 bOverrideConcurrency=true
 maxDistance=5300
 soundClassObject=&Fort_Impacts
 volumeMultiplier=0.8437
 attenuationOverrides {falloffDistance=2200}
 concurrencyOverrides {maxCount=3}
 concurrencySet [&RifleBulletImpact_Conc]
}
