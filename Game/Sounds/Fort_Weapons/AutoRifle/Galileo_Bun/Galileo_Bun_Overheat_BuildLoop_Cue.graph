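# Looping overheat build-up sound cue for the Galileo_Bun auto rifle weapon.
# A single looping wave runs through a continuous modulator that maps one Overheat parameter to both pitch and volume, so the sound tracks the heat value.
# It plays as weapon foley with reload attenuation and overrides concurrency with a small owner-limited cap that stops the oldest instance.
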SoundCue /Game/Sounds/Fort_Weapons/AutoRifle/Galileo_Bun/Galileo_Bun_Overheat_BuildLoop_Cue
root=ModulatorContinuous_0
nodes [
 {
  id=ModulatorContinuous_0
  children [WavePlayer_0]
  properties {
   pitchModulationParams {default=0.5, maxOutput=3, parameterName=Overheat}
   volumeModulationParams {default=0, minInput=0.1, minOutput=0.4, parameterName=Overheat}
  }
 }
 {
  id=WavePlayer_0
  properties {bLooping=true, soundWaveAssetPtr=&/Galileo_Bun_Overheat_BuildLoop_01}
 }
]
properties {
 attenuationSettings=&Fort_Weapon_Reload_Att
 bDebug=true
 bOverrideConcurrency=true
 maxDistance=2250
 soundClassObject=&Fort_WeaponFoley
 volumeMultiplier=0.501
 concurrencyOverrides {bLimitToOwner=true, maxCount=2, resolutionRule=StopOldest}
}
