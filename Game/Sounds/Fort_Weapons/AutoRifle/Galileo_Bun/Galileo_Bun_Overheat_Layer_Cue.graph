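# Sound cue for the overheat audio layer of the Galileo_Bun auto rifle weapon.
# A mixer blends a random pick from three equally weighted waves with a fourth wave, which runs through a modulator and is mixed much louder.
# The cue plays at a slightly lowered pitch in a ranged-weapon low-ammo/degraded sound class.
# Its own concurrency override blocks new instances once the limit is reached.
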SoundCue /Game/Sounds/Fort_Weapons/AutoRifle/Galileo_Bun/Galileo_Bun_Overheat_Layer_Cue
root=Mixer_0
nodes [
 {id=Mixer_0, children [Random_0,Modulator_0], properties {inputVolume [0.75,2.5]}}
 {id=Modulator_0, children [WavePlayer_4]}
 {id=Random_0, children [WavePlayer_0,WavePlayer_2,WavePlayer_3], properties {weights [1,1,1]}}
 {id=WavePlayer_0, properties {soundWaveAssetPtr=&/Galileo_Bun_Overheat_Layer01}}
 {id=WavePlayer_2, properties {soundWaveAssetPtr=&/Galileo_Bun_Overheat_Layer02}}
 {id=WavePlayer_3, properties {soundWaveAssetPtr=&/Galileo_Bun_Overheat_Layer03}}
 {id=WavePlayer_4, properties {soundWaveAssetPtr=&/Galileo_Bun_Overheat_Layer04}}
]
properties {
 bOverrideConcurrency=true
 maxDistance=2097152
 pitchMultiplier=0.9
 soundClassObject=/Game/Sounds/FortSoundClassesAndModes/Fort_WeaponRanged_LowAmmo_Degraded_PSM
 concurrencyOverrides {resolutionRule=PreventNew}
}
