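# Sound cue for the overheat release of the Galileo_Bun auto rifle, played as weapon foley.
# A mixer layers a randomly picked, modulated release variant with a quieter mounted-turret overheat sound and an Ashton Milo drop sound reused from other content.
# It uses weapon reload attenuation and a per-owner concurrency override that stops the oldest instance.
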SoundCue /Game/Sounds/Fort_Weapons/AutoRifle/Galileo_Bun/Galileo_Bun_Overheat_Release_Cue
root=Mixer_0
nodes [
 {id=Mixer_0, children [Modulator_2,WavePlayer_4,WavePlayer_3], properties {inputVolume [1,0.2,1]}}
 {id=Modulator_2, children [Random_0]}
 {id=Random_0, children [WavePlayer_0,WavePlayer_1,WavePlayer_2], properties {weights [1,1,1]}}
 {id=WavePlayer_0, properties {soundWaveAssetPtr=&/Galileo_Bun_Overheat_Release_01}}
 {id=WavePlayer_1, properties {soundWaveAssetPtr=&/Galileo_Bun_Overheat_Release_02}}
 {id=WavePlayer_2, properties {soundWaveAssetPtr=&/Galileo_Bun_Overheat_Release_03}}
 {id=WavePlayer_3, properties {soundWaveAssetPtr=/Game/Athena/Sounds/LTMs/Ashton/Milo/Ashton_Milo_BR_Drop}}
 {id=WavePlayer_4, properties {soundWaveAssetPtr=/Game/Sounds/Fort_Traps/MountedTurret/Turret_Ballistic_Overheat_02}}
]
properties {
 attenuationSettings=&Fort_Weapon_Reload_Att
 bDebug=true
 bOverrideConcurrency=true
 maxDistance=2250
 soundClassObject=&Fort_WeaponFoley
 volumeMultiplier=0.501
 concurrencyOverrides {bLimitToOwner=true, maxCount=2, resolutionRule=StopOldest}
}
