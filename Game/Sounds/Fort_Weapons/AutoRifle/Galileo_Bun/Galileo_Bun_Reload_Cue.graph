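# Sound cue for the reload foley of the Galileo_Bun auto rifle weapon.
# A single wave player feeds a short fixed delay at the root, so the same reload sample plays slightly late on every trigger, with no randomization.
# The cue uses the shared weapon reload attenuation and WeaponFoley sound class, slightly lowers pitch and volume, overrides concurrency, and has its debug flag enabled.
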SoundCue /Game/Sounds/Fort_Weapons/AutoRifle/Galileo_Bun/Galileo_Bun_Reload_Cue
root=Delay_0
nodes [
 {id=Delay_0, children [WavePlayer_4], properties {delayMax=0.04, delayMin=0.04}}
 {id=WavePlayer_4, properties {soundWaveAssetPtr=&/Galileo_Bun_Reload_05}}
]
properties {
 attenuationSettings=&Fort_Weapon_Reload_Att
 bDebug=true
 bOverrideConcurrency=true
 maxDistance=2250
 pitchMultiplier=0.95
 soundClassObject=&Fort_WeaponFoley
 volumeMultiplier=0.752
}
