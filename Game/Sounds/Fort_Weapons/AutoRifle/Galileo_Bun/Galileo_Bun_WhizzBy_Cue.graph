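# Sound cue for bullet whizz-by sounds of the Galileo_Bun auto rifle, in the ranged weapon sound class.
# A weighted random node picks one of three waves, favoring the 05 variant, which gets its own wider pitch modulation.
# An outer modulator then varies pitch on whichever wave plays.
# Attenuation is overridden to disable distance falloff with a wide stereo spread, and a whizz-by concurrency rule governs simultaneous plays.
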SoundCue /Game/Sounds/Fort_Weapons/AutoRifle/Galileo_Bun/Galileo_Bun_WhizzBy_Cue
root=Modulator_0
nodes [
 {id=Modulator_0, children [Random_1], properties {pitchMax=1.1, pitchMin=0.9}}
 {id=Modulator_1, children [WavePlayer_8], properties {pitchMax=1.1, pitchMin=0.8}}
 {id=Random_1, children [WavePlayer_5,WavePlayer_6,Modulator_1], properties {weights [1,1,2]}}
 {id=WavePlayer_5, properties {soundWaveAssetPtr=&/Galileo_Bun_WhizzBy_02}}
 {id=WavePlayer_6, properties {soundWaveAssetPtr=&/Galileo_Bun_WhizzBy_03}}
 {id=WavePlayer_8, properties {soundWaveAssetPtr=&/Galileo_Bun_WhizzBy_05}}
]
properties {
 bOverrideAttenuation=true
 maxDistance=2097152
 soundClassObject=&Fort_WeaponRanged
 volumeMultiplier=0.3
 attenuationOverrides {bAttenuate=false, bUseComplexCollisionForOcclusion=true, occlusionVolumeAttenuation=0, stereoSpread=5000}
 concurrencySet [/Game/Sounds/Concurrency_Rules/Athena_BulletWhizBy_Conc]
}
